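# Parent material for the Elder Pro vehicle cosmetic wheel, adding a shimmering, noise-driven rim effect to the MF_WheelParent_Simple base.
# Reflection vector, base normal masked by inverted red vertex color, and a panning flow normal form noise UVs.
# That noise, multiplied by static tiled noise and boosted heavily, feeds lerped rim color and emissive reroutes.
# Base Color and Emissive are overridden through reroutes before vehicle damage and de-rez transition functions.
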Material /VehicleCosmetics/Wheels/Elder_Pro/Materials/M_WheelParent_Elder_Pro
graph {
 properties {
  bAutomaticallySetUsageInEditor=false
  bUseMaterialAttributes=true
  bUsedWithMeshDeformer=true
  bUsedWithSkeletalMesh=true
  pixelDepthOffsetMode=PDOM_Legacy
  usedShadingModels=Default Lit
 }
 nodes [
  n1 = MaterialFunctionCall() {materialFunction=&MF_WheelParent_Simple}
  n2 = GetMaterialAttributes(None=@n1.MaterialAttributes) {attributeGetTypes [#1, #2, #3]}
  n3 = NamedRerouteUsage()
  n4 = NamedRerouteUsage()
  MaterialFunctionCall(In=MaterialFunctionCall(Material=SetMaterialAttributes!(MaterialAttributes=@n2.MaterialAttributes, Base Color=@n3, Emissive Color=@n4, Normal) {attributeSetTypes [#1, #2, #3]}) {materialFunction=&MF_BaseVehicle_Damage}.Result) {materialFunction=&MF_VehicleDeRez_Transition}
  NamedRerouteDeclaration(None=@n2.Base Color) {name=BC, nodeColor=<rgba 1 0 0.3984 1>}
  NamedRerouteDeclaration(None=@n2.Emissive Color) {name=Emissive, nodeColor=<rgba 0.6172 0 1 1>}
  n5 = OneMinus(None=NamedRerouteDeclaration(None=@n1.VertexColorR) {name=VC_R, nodeColor=<rgba 1 0 0.4219 1>})
  n6 = TextureSample(UVs=Panner() {speedY=0.005}) {
   samplerSource=SSM_Wrap_WorldGroupSettings
   samplerType=SAMPLERTYPE_Normal
   texture=/VehicleCosmetics/Wheels/_SharedTextures/FX/Flow_03_nomips
  }
  n7 = Add(A=ReflectionVectorWS(), B=Add(A=Multiply(A=@n2.Normal, B=@n5), B=@n6.RGB))
  n8 = TextureSample(UVs=Multiply(A=ComponentMask(None=@n7) {g=true, r=true}) {constB=0.0625}) {samplerSource=SSM_Wrap_WorldGroupSettings, texture=/VehicleCosmetics/Wheels/_SharedTextures/FX/Noise_Generic_01_Pack}
  n9 = TextureSample(UVs=TextureCoordinate() {uTiling=4, vTiling=4}) {samplerSource=SSM_Wrap_WorldGroupSettings, texture=&Noise_Generic_01_Pack}
  n10 = NamedRerouteUsage()
  n11 = Multiply(A=Multiply(A=Multiply(A=@n8.B, B=@n9.B)) {constB=4096}, B=@n10)
  NamedRerouteDeclaration(None=LinearInterpolate(A=@n11, B=NamedRerouteUsage(), Alpha=NamedRerouteUsage())) {name=Rim Emissive, nodeColor=<rgba 0.5 0 1 1>}
  NamedRerouteDeclaration(None=LinearInterpolate(A=Add(A=@n11, B=@n10), B=@n10, Alpha=NamedRerouteUsage())) {name=Rim BC, nodeColor=<rgba 1 0.2578 0 1>}
 ]
}
