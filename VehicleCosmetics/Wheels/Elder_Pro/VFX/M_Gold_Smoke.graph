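# Translucent lit gold smoke material for Niagara sprite, ribbon and mesh particles in the Elder_Pro wheel cosmetic VFX.
# A panning cloud texture distorts the UVs of a flame-cloud texture, while low quality swaps in an undistorted sample.
# Color multiplies that texture by particle color and a gold-weighted lerp toward the Brightness dynamic parameter, adding a floored copy.
# Particle alpha times the texture is heavily scaled and clamped to form an opacity-style mask.
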Material /VehicleCosmetics/Wheels/Elder_Pro/VFX/M_Gold_Smoke
graph {
 properties {
  bAutomaticallySetUsageInEditor=false
  bUsedWithNiagaraMeshParticles=true
  bUsedWithNiagaraRibbons=true
  bUsedWithNiagaraSprites=true
  blendMode=BLEND_Translucent
  usedShadingModels=Default Lit
 }
 nodes [
  n1 = LinearInterpolate(
   A=Constant3Vector() {constant=<rgba 0.0918 0.0918 0.0918 0>}
   B=DynamicParameter() {paramNames [Param1,Brightness,Param3,Param4]}.Brightness
   Alpha=Constant3Vector() {constant=<rgba 0.5801 0.3578 0.0769 0>}
  )
  n2 = TextureCoordinate() {uTiling=0.25, vTiling=0.25}
  n3 = Multiply(A=Time(), B=Constant() {r=0.5})
  n4 = ComponentMask(None=TextureSample(UVs=Panner(Coordinate=@n2, Time=@n3) {speedX=0.1, speedY=0.1}) {texture=&/Clouds_RGB_Soft_T}.RGB) {g=true, r=true}
  n5 = Constant() {r=0.1}
  n6 = TextureCoordinate()
  n7 = TextureSample(UVs=Add(A=Multiply(A=@n4, B=@n5), B=@n6)) {texture=/VehicleCosmetics/Wheels/SnowFlake/VFX/FlameCloud_01}
  n8 = Clamp(None=QualitySwitch(Default=@n7.RGB, Low=TextureSample(UVs=@n6) {texture=&/FlameCloud_02}.RGB, High=@n7.RGB, Medium=@n7.RGB))
  n9 = ParticleColor()
  n10 = Multiply(A=@n8, B=@n9.RGB)
  Add(A=Multiply(A=@n1, B=@n10), B=Multiply(A=Clamp(None=@n10) {minDefault=0.16}, B=@n1))
  Clamp(None=ConstantBiasScale(None=Multiply(A=@n9.A, B=@n8)) {bias=0, scale=20})
 ]
}
