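# Translucent Niagara particle material for the Elder Pro wheel cosmetic, producing a bright orange, flame-like effect on sprites, meshes and ribbons.
# Two noise textures panning at different speeds multiply with a flame cloud texture, then clamp and scale by particle alpha to form an animated mask.
# A separate constant HDR orange supplies the color.
# Quality switches replace both mask and color with zero at Low quality.
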Material /VehicleCosmetics/Wheels/Elder_Pro/VFX/M_NS_Wheel_ElderPro
graph {
 properties {
  bAutomaticallySetUsageInEditor=false
  bGenerateSphericalParticleNormals=true
  bUsedWithMeshParticles=true
  bUsedWithNiagaraMeshParticles=true
  bUsedWithNiagaraRibbons=true
  bUsedWithNiagaraSprites=true
  bUsedWithParticleSprites=true
  blendMode=BLEND_Translucent
  previewMesh=/VehicleCosmetics/Wheels/CobaltDagger/Meshes/SK_Wheel_CobaltDagger
  usedShadingModels=Default Lit
 }
 nodes [
  n1 = Panner(Coordinate=TextureCoordinate() {uTiling=6, vTiling=6}) {speedY=0.2}
  ParticleSubUV!(UVs=@n1, Tex, Apply View MipBias) {texture=/VehicleCosmetics/Wheels/_SharedTextures/FX/Universal_Noise_32_RGB}
  n2 = TextureCoordinate()
  ParticleSubUV(UVs=@n2) {texture=&/FlameCloud_02}
  n3 = Time()
  n4 = Constant() {r=2}
  n5 = TextureSample(UVs=Panner(Coordinate=TextureCoordinate() {uTiling=0.5, vTiling=0.5}, Time=Multiply(A=@n3, B=@n4)) {speedX=1, speedY=0.001}) {texture=/VehicleCosmetics/Wheels/_SharedTextures/FX/Universal_Noise_42_RGB}
  n6 = Multiply(
   A=Multiply(
    A=ConstantBiasScale(None=@n5.R) {bias=-0.25, scale=2}
    B=TextureSample(UVs=@n1) {texture=&Universal_Noise_32_RGB}.R
   )
   B=ConstantBiasScale(None=TextureSample(UVs=@n2) {texture=&FlameCloud_02}.R) {bias=0, scale=1}
  )
  n7 = ConstantBiasScale(None=Multiply(A=Clamp(None=@n6), B=ParticleColor().A)) {bias=0, scale=1}
  n8 = Constant()
  QualitySwitch(Default=@n7, Low=@n8, High=@n7)
  n9 = Multiply(A=Constant3Vector() {constant=<rgba 1 0.5559 0.2399 0>}, B=Constant() {r=50})
  QualitySwitch(Default=@n9, Low=@n8, High=@n9)
  VertexColor()
 ]
}
